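# Unlit alpha-composite material that raymarches a distance field stored in a flipbook-style pseudo-volume texture, optionally showing its RGB (gradient) as color.
# VolumeBoxIntersect supplies the entry point and box distance to a Custom HLSL loop with optional per-pixel, per-frame jitter.
# Alpha below DF Center becomes density, accumulated with exponential absorption into color and opacity.
# Parameter A blends white toward the sampled gradient; no lighting is computed, and the FinalStep input is unused.
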Material /Volumetrics/Content/VolumeTextures/Materials/Raymarching/M_VolumeRayMarch_DistanceField_Gradient
graph {
 properties {bDisableDepthTest=true, blendMode=BLEND_AlphaComposite, pixelDepthOffsetMode=PDOM_Legacy, shadingModel=MSM_Unlit, usedShadingModels=Unlit}
 nodes [
  MaterialExpressionTextureObjectParameter_0 = TextureObjectParameter() {parameterName=RT, texture=/Volumetrics/Content/VolumeTextures/Textures/RT_Preview}
  MaterialExpressionComponentMask_8 = ComponentMask(None=VectorParameter() {parameterName=XYFrames, defaultValue=<rgba 16 16 0 0>}.RGB) {g=true, r=true}
  MaterialExpressionDivide_0 = Divide(B=ScalarParameter() {defaultValue=256, parameterName=Steps})
  n1 = MaterialFunctionCall!(
   Steps=ScalarParameter() {defaultValue=256, parameterName=Steps}
   Plane alignment.=ScalarParameter() {defaultValue=1, parameterName=Plane Alignment}
  ) {materialFunction=/Volumetrics/Content/VolumeTextures/MaterialFunctions/VolumeBoxIntersect}
  n2 = Multiply(A=@n1.Box Distance, B=ScalarParameter() {defaultValue=256, parameterName=Steps})
  MaterialExpressionClamp_1 = Clamp(None=Floor(None=@n2)) {desc=MaxSteps, maxDefault=512}
  MaterialExpressionFrac_0 = Frac(None=@n2) {desc=FinalStepSize}
  MaterialExpressionMultiply_1 = Multiply(A=@n1.Intersection Position) {desc=Ray Entry}
  MaterialExpressionScalarParameter_7 = ScalarParameter() {defaultValue=32, parameterName=Density}
  MaterialExpressionScalarParameter_11 = ScalarParameter() {defaultValue=0.53, parameterName=DF Center}
  MaterialExpressionScalarParameter_8 = ScalarParameter() {parameterName=Jitter}
  MaterialExpressionScalarParameter_0 = ScalarParameter() {parameterName=A}
  n3 = Custom!(
   Tex=@MaterialExpressionTextureObjectParameter_0
   XYFrames=@MaterialExpressionComponentMask_8
   StepSize=@MaterialExpressionDivide_0
   MaxSteps=@MaterialExpressionClamp_1
   FinalStep=@MaterialExpressionFrac_0
   CurPos=@MaterialExpressionMultiply_1
   Density=@MaterialExpressionScalarParameter_7
   DFCenter=@MaterialExpressionScalarParameter_11
   Jitter=@MaterialExpressionScalarParameter_8
   ShowGradient=@MaterialExpressionScalarParameter_0
  ) {
   code="float numFrames = XYFrames.x * XYFrames.y;\r\nfloat accumdist = 0;\r\nfloat curdensity = 0;\r\nfloat transmittance = 1;\r\nfloat3 localcamvec = normalize( mul(Parameters.CameraVector, (float3x3)WSDemote(GetWorldToLocal(Parameters))) ) * StepSize;\r\n\r\n\r\nDensity *= StepSize;\r\nfloat3 lightenergy = 0;\r\n\r\n\r\nint3 randpos = int3(Parameters.SvPosition.xy, View.StateFrameIndexMod8);\r\nfloat rand =float(Rand3DPCG16(randpos).x) / 0xffff;\r\nCurPos +=  localcamvec * rand.x * Jitter;\r\n\r\n\r\nfor (int i = 0; i < MaxSteps; i++)\r\n{\t\r\n\r\n\t\r\n\tfloat4 cursample = PseudoVolumeTexture(Tex, TexSampler, CurPos, XYFrames, numFrames);\r\n\tcurdensity = -(cursample.a - DFCenter) * Density;\r\n\t\r\n\t//Sample Light Absorption and Scattering\r\n\tif( (curdensity) > 0.001)\r\n\t{\r\n\t\tfloat3 lpos = CurPos;\r\n\t\tfloat shadowdist = 0;\r\n\t\r\n\t\tcursample.rgb = lerp(1, cursample.rgb, ShowGradient);\r\n\t\r\n\t\tcurdensity = 1 - exp(-curdensity);\r\n\r\n\t\tlightenergy += cursample.rgb * curdensity * transmittance;\r\n\r\n\t\ttransmittance *= 1 - curdensity;\r\n\r\n\t}\r\n\t\t\r\n\r\n\tCurPos -= localcamvec;\r\n}\r\n\r\n\r\n\r\nreturn float4( lightenergy, transmittance);\r\n"
   desc="\n*** Original source before expression upgrade ***\nfloat numFrames = XYFrames.x * XYFrames.y;\r\nfloat accumdist = 0;\r\nfloat curdensity = 0;\r\nfloat transmittance = 1;\r\nfloat3 localcamvec = normalize( mul(Parameters.CameraVector, (float3x3)GetPrimitiveData(Parameters).WorldToLocal) ) * StepSize;\r\n\r\n\r\nDensity *= StepSize;\r\nfloat3 lightenergy = 0;\r\n\r\n\r\nint3 randpos = int3(Parameters.SvPosition.xy, View.StateFrameIndexMod8);\r\nfloat rand =float(Rand3DPCG16(randpos).x) / 0xffff;\r\nCurPos +=  localcamvec * rand.x * Jitter;\r\n\r\n\r\nfor (int i = 0; i < MaxSteps; i++)\r\n{\t\r\n\r\n\t\r\n\tfloat4 cursample = PseudoVolumeTexture(Tex, TexSampler, CurPos, XYFrames, numFrames);\r\n\tcurdensity = -(cursample.a - DFCenter) * Density;\r\n\t\r\n\t//Sample Light Absorption and Scattering\r\n\tif( (curdensity) > 0.001)\r\n\t{\r\n\t\tfloat3 lpos = CurPos;\r\n\t\tfloat shadowdist = 0;\r\n\t\r\n\t\tcursample.rgb = lerp(1, cursample.rgb, ShowGradient);\r\n\t\r\n\t\tcurdensity = 1 - exp(-curdensity);\r\n\r\n\t\tlightenergy += cursample.rgb * curdensity * transmittance;\r\n\r\n\t\ttransmittance *= 1 - curdensity;\r\n\r\n\t}\r\n\t\t\r\n\r\n\tCurPos -= localcamvec;\r\n}\r\n\r\n\r\n\r\nreturn float4( lightenergy, transmittance);\r\n"
   description=Density RayMarch
   outputType=CMOT_Float4
   inputs [
    {
     inputName=Tex
     input {expression=&self:MaterialExpressionTextureObjectParameter_0}
    }
    {
     inputName=XYFrames
     input {expression=&self:MaterialExpressionComponentMask_8}
    }
    {
     inputName=StepSize
     input {expression=&self:MaterialExpressionDivide_0}
    }
    {
     inputName=MaxSteps
     input {expression=&self:MaterialExpressionClamp_1}
    }
    {
     inputName=FinalStep
     input {expression=&self:MaterialExpressionFrac_0}
    }
    {
     inputName=CurPos
     input {expression=&self:MaterialExpressionMultiply_1}
    }
    {
     inputName=Density
     input {expression=&self:MaterialExpressionScalarParameter_7}
    }
    {
     inputName=DFCenter
     input {expression=&self:MaterialExpressionScalarParameter_11}
    }
    {
     inputName=Jitter
     input {expression=&self:MaterialExpressionScalarParameter_8}
    }
    {
     inputName=ShowGradient
     input {expression=&self:MaterialExpressionScalarParameter_0}
    }
   ]
  }
  ComponentMask(None=@n3) {b=true, g=true, r=true}
  OneMinus(None=ComponentMask(None=@n3) {a=true})
 ]
}
