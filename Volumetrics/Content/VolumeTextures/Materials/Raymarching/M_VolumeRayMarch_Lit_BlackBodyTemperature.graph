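# Unlit alpha-composite material that ray-marches a lit, glowing volume inside a box, with depth testing disabled.
# VolumeBoxIntersect feeds a Custom HLSL loop sampling a flipbook-packed Color texture: RGB as density, alpha scaling blackbody temperature for emission.
# Samples add shadow-marched sunlight with Schlick phase and three-tap sky occlusion, yielding capped light and inverted transmittance.
# Curl-noise distortion using Velocity and Pressure textures is wired but compiled out by EnableCurl 0.
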Material /Volumetrics/Content/VolumeTextures/Materials/Raymarching/M_VolumeRayMarch_Lit_BlackBodyTemperature
graph {
 properties {bDisableDepthTest=true, blendMode=BLEND_AlphaComposite, pixelDepthOffsetMode=PDOM_Legacy, shadingModel=MSM_Unlit, usedShadingModels=Unlit}
 nodes [
  n1 = MaterialFunctionCall!(
   Steps=ScalarParameter() {defaultValue=128, parameterName=Steps}
   Plane alignment.=ScalarParameter() {defaultValue=1, parameterName=Plane Alignment}
  ) {materialFunction=/Volumetrics/Content/VolumeTextures/MaterialFunctions/VolumeBoxIntersect}
  n2 = Multiply(A=@n1.Box Distance, B=ScalarParameter() {defaultValue=128, parameterName=Steps})
  Ceil(None=@n2)
  n3 = ScalarParameter() {defaultValue=5, parameterName=Min}
  MaterialExpressionTextureObjectParameter_1 = TextureObjectParameter() {parameterName=Color, texture=/Volumetrics/Content/VolumeTextures/Textures/RT_Default}
  MaterialExpressionComponentMask_4 = ComponentMask(None=VectorParameter() {parameterName=XYFrames, defaultValue=<rgba 10 10 0 0>}.RGB) {g=true, r=true}
  MaterialExpressionDivide_8 = Divide(B=ScalarParameter() {defaultValue=128, parameterName=Steps})
  MaterialExpressionClamp_0 = Clamp(None=Floor(None=@n2)) {desc=MaxSteps, maxDefault=512}
  MaterialExpressionFrac_3 = Frac(None=@n2) {desc=FinalStepSize}
  MaterialExpressionMultiply_0 = Multiply(A=@n1.Intersection Position) {desc=Ray Entry}
  MaterialExpressionScalarParameter_0 = ScalarParameter() {defaultValue=16, parameterName=Density}
  MaterialExpressionScalarParameter_25 = ScalarParameter() {defaultValue=16, parameterName=ShadowSteps}
  MaterialExpressionScalarParameter_26 = ScalarParameter() {defaultValue=4, parameterName=ShadowDensity}
  MaterialExpressionNormalize_1 = Normalize(VectorInput=Transform(None=VectorParameter() {parameterName=LightVector, defaultValue=<rgba 1 0 1 1>}.RGB) {transformSourceType=TRANSFORMSOURCE_World, transformType=TRANSFORM_Local})
  MaterialExpressionScalarParameter_12 = ScalarParameter() {defaultValue=0.01, parameterName=SB}
  n4 = VectorParameter() {parameterName=Sun Color, defaultValue=<rgba 1 0.875 0.55 4>}
  n5 = DotProduct(A=Normalize(VectorInput=VectorParameter() {parameterName=LightVector, defaultValue=<rgba 1 0 1 1>}.RGB), B=CameraVectorWS())
  n6 = ScalarParameter() {defaultValue=0.3, parameterName=Aniso}
  MaterialExpressionMultiply_17 = Multiply(
   A=@n4.RGB
   B=Multiply(
    A=@n4.A
    B=MaterialFunctionCall!(Cosine=@n5, Anisotropy=@n6) {materialFunction=/Engine/Functions/Engine_MaterialFunctions01/Volumetrics/SchlickPhase}.Result
   )
  )
  n7 = VectorParameter() {parameterName=Sky Color, defaultValue=<rgba 0.2 0.2 0.25 0.2>}
  MaterialExpressionMultiply_19 = Multiply(A=@n7.RGB, B=@n7.RGB)
  MaterialExpressionScalarParameter_4 = ScalarParameter() {defaultValue=0.2, parameterName=Ambient Density}
  MaterialExpressionMultiply_3 = Multiply(
   A=MaterialFunctionCall(Alpha Threshold=ScalarParameter() {defaultValue=0.5, parameterName=Jitter Alpha Threshold}) {materialFunction=&DitherTemporalAA}.Result
   B=ScalarParameter() {defaultValue=1, parameterName=Jitter}
  )
  MaterialExpressionTextureObjectParameter_2 = TextureObjectParameter() {parameterName=Velocity, texture=&RT_Default}
  MaterialExpressionScalarParameter_27 = ScalarParameter() {defaultValue=10000, parameterName=Temperature}
  MaterialExpressionTextureObject_0 = TextureObject() {samplerType=SAMPLERTYPE_LinearColor, texture=/Volumetrics/Content/VolumeTextures/Textures/T_CurlNoise_Volume_Tiling_01}
  MaterialExpressionScalarParameter_29 = ScalarParameter() {defaultValue=0.4, parameterName=Curl Tiling}
  MaterialExpressionScalarParameter_30 = ScalarParameter() {defaultValue=0.03, parameterName=Curl Amount}
  MaterialExpressionMultiply_20 = Multiply(
   A=VectorParameter() {parameterName=Panner, defaultValue=<rgba 0 1 -1 0>}.RGB
   B=MaterialFunctionCall(Speed=ScalarParameter() {parameterName=Speed}) {materialFunction=&TimeWithSpeedVariable}.Time
  )
  MaterialExpressionTextureObjectParameter_3 = TextureObjectParameter() {parameterName=Pressure, texture=&RT_Default}
  MaterialExpressionScalarParameter_7 = ScalarParameter() {defaultValue=0.5, parameterName=Pressure Display}
  MaterialExpressionScalarParameter_3 = ScalarParameter() {defaultValue=0.025, parameterName=CurlAdvection}
  MaterialExpressionScalarParameter_9 = ScalarParameter() {defaultValue=0.75, parameterName=Light Offset}
  MaterialExpressionVectorParameter_0 = VectorParameter() {parameterName=VolumeResolution, defaultValue=<rgba 256 256 256 0>}
  n8 = Custom!(
   Tex=@MaterialExpressionTextureObjectParameter_1
   XYFrames=@MaterialExpressionComponentMask_4
   StepSize=@MaterialExpressionDivide_8
   MaxSteps=@MaterialExpressionClamp_0
   FinalStep=@MaterialExpressionFrac_3
   CurPos=@MaterialExpressionMultiply_0
   Density=@MaterialExpressionScalarParameter_0
   ShadowSteps=@MaterialExpressionScalarParameter_25
   ShadowDensity=@MaterialExpressionScalarParameter_26
   LightVector=@MaterialExpressionNormalize_1
   ShadowThreshold=@MaterialExpressionScalarParameter_12
   LightColor=@MaterialExpressionMultiply_17
   SkyColor=@MaterialExpressionMultiply_19
   AmbientDensity=@MaterialExpressionScalarParameter_4
   Jitter=@MaterialExpressionMultiply_3
   Tex2=@MaterialExpressionTextureObjectParameter_2
   Temp=@MaterialExpressionScalarParameter_27
   CurlTex=@MaterialExpressionTextureObject_0
   CurlTiling=@MaterialExpressionScalarParameter_29
   CurlAmount=@MaterialExpressionScalarParameter_30
   CurlPan=@MaterialExpressionMultiply_20
   Tex3=@MaterialExpressionTextureObjectParameter_3
   PressureDisplay=@MaterialExpressionScalarParameter_7
   curladvect=@MaterialExpressionScalarParameter_3
   LightOffset=@MaterialExpressionScalarParameter_9
   VolumeResolution=@MaterialExpressionVectorParameter_0.RGB
  ) {
   code="##define EnableCurl 0\r\nfloat numFrames = XYFrames.x * XYFrames.y;\r\nfloat accumdist = 0;\r\nfloat3 curdensity = 0;\r\nfloat transmittance = 1;\r\nfloat3 localcamvec = normalize( mul(Parameters.CameraVector, (float3x3)WSDemote(GetWorldToLocal(Parameters))) ) * StepSize;\r\n\r\nfloat3 invlightdir = 1 / LightVector;\r\n\r\nfloat shadowstepsize = 1 / ShadowSteps;\r\nLightVector *= shadowstepsize*0.25;\r\nShadowDensity *= shadowstepsize;\r\n\r\nDensity *= StepSize;\r\nfloat3 lightenergy = 0;\r\nfloat shadowthresh = -log(ShadowThreshold) / ShadowDensity;\r\n\r\nint3 randpos = int3(Parameters.SvPosition.xy, View.StateFrameIndexMod8);\r\nfloat rand =float(Rand3DPCG16(randpos).x) / 0xffff;\r\nrand=1;\r\nCurPos +=  localcamvec * rand.x * Jitter;\r\n\r\n\r\nfor (int i = 0; i < MaxSteps; i++)\r\n{\t\r\n\r\n\r\n\t##if EnableCurl\r\n\t\tfloat4 velocity = PseudoVolumeTexture(Tex2, TexSampler, saturate(CurPos), XYFrames, numFrames);\r\n\t\tfloat3 curlpos = CurPos * CurlTiling + CurlPan;\r\n\t\tcurlpos += normalize(velocity) * curladvect;\r\n\t\tfloat3 curl = PseudoVolumeTexture(CurlTex, TexSampler, curlpos, 16, 256).rgb;\r\n\t\tfloat pressure = PseudoVolumeTexture(Tex3, TexSampler, saturate(CurPos), XYFrames, numFrames).r;\r\n\t\tpressure = saturate( -pressure+PressureDisplay);\r\n\t\t\r\n\t\tcurl = CurlAmount * 2 * (curl - 0.5);\r\n\t\tcurl *= pressure;\r\n\t##else\r\n\t\tfloat3 curl = 0;\r\n\t##endif\r\n\r\n\tfloat3 samplepos =  saturate(CurPos + curl);\r\n\tsamplepos.z = min(samplepos.z, 1 - (1/VolumeResolution.z));\r\n\r\n\tfloat4 cursample = PseudoVolumeTexture(Tex, TexSampler, samplepos, XYFrames, numFrames);\r\n\t//float pressure = max(0, PseudoVolumeTexture(Tex3, TexSampler, saturate(CurPos), XYFrames, numFrames).r);\r\n\r\n\r\n\t//Sample Light Absorption and Scattering\r\n\tif( cursample.r + cursample.g + cursample.b > 0.001)\r\n\t{\r\n\t\tfloat3 lpos = CurPos - LightVector * LightOffset;\r\n\t\tfloat shadowdist = 0;\r\n\r\n\t\tfor (int s = 0; s < ShadowSteps; s++)\r\n\t\t{\r\n\t\t\tlpos += LightVector;\r\n\r\n\t\t\t##if EnableCurl\r\n\t\t\t\tvelocity = PseudoVolumeTexture(Tex2, TexSampler, saturate(lpos), XYFrames, numFrames);\r\n\t\t\t\tcurlpos = lpos * CurlTiling + CurlPan;\r\n\t\t\t\tcurlpos += normalize(velocity) * curladvect;\r\n\t\t\t\tcurl = PseudoVolumeTexture(CurlTex, TexSampler, curlpos , 16, 256).rgb;\r\n\t\t\t\tcurl = CurlAmount * 2 * (curl - 0.5);\r\n\r\n\t\t\t\tpressure = PseudoVolumeTexture(Tex3, TexSampler, saturate(CurPos), XYFrames, numFrames).r;\r\n\t\t\t\tpressure = saturate( -pressure+PressureDisplay);\r\n\t\t\t\tcurl *= pressure;\r\n\t\t\t##else\r\n\t\t\t\tcurl = 0;\r\n\t\t\t##endif\r\n\t\t\r\n\t\t\tfloat3 lsample = PseudoVolumeTexture(Tex, TexSampler, saturate(lpos + curl), XYFrames, numFrames).rgb;\r\n\t\t\t\r\n\t\t\tfloat3 shadowboxtest = floor( 0.5 + ( abs( 0.5 - lpos ) ) );\r\n\t\t\tfloat exitshadowbox = shadowboxtest .x + shadowboxtest .y + shadowboxtest .z;\r\n\r\n       \t\t\tif(shadowdist > shadowthresh || exitshadowbox >= 1) break;\r\n\r\n\t\t\tshadowdist += dot(lsample, 0.3333);\r\n\t\t}\r\n\t\r\n\t\tcurdensity = 1 - exp(-cursample.rgb * Density);\r\n\t\tlightenergy += exp(-shadowdist * ShadowDensity) * curdensity * transmittance * LightColor;\r\n\t\tlightenergy += (MaterialExpressionBlackBody(Temp*cursample.a)) * curdensity * transmittance;\r\n\t\t//lightenergy += pressure * transmittance * PressureDisplay;\r\n\r\n\t\t\r\n\t\ttransmittance *= 1-dot(curdensity, 0.333);\r\n\t\t##if 1\r\n\r\n\t\t//Sky Lighting\r\n\t\tshadowdist = 0;\r\n\t\t\r\n\t\tlpos = CurPos + float3(0,0,0.025);\r\n\t\tfloat3 lsample = PseudoVolumeTexture(Tex, TexSampler, saturate(lpos), XYFrames, numFrames).rgb;\r\n\t\tshadowdist += dot(lsample, 0.3333);\r\n\t\tlpos = CurPos + float3(0,0,0.05);\r\n\t\tlsample = PseudoVolumeTexture(Tex, TexSampler, saturate(lpos), XYFrames, numFrames).rgb;\r\n\t\tshadowdist += dot(lsample, 0.3333);\r\n\t\tlpos = CurPos + float3(0,0,0.15);\r\n\t\tlsample = PseudoVolumeTexture(Tex, TexSampler, saturate(lpos), XYFrames, numFrames).rgb;\r\n\t\tshadowdist += dot(lsample, 0.3333);\r\n\r\n\t\t//shadowterm    = exp(-shadowdist  * AmbientDensity);\r\n\t\t//absorbedlight = exp(-shadowdist  * AmbientDensity) * curdensity;\r\n\t\tlightenergy    += exp(-shadowdist  * AmbientDensity) * curdensity * SkyColor * transmittance;\r\n\t\t##endif\r\n\r\n\t}\r\n\t\t\r\n\r\n\tCurPos -= localcamvec;\r\n}\r\n\r\n\r\n\r\nreturn float4( lightenergy, transmittance);\r\n"
   desc="\n*** Original source before expression upgrade ***\n##define EnableCurl 0\r\nfloat numFrames = XYFrames.x * XYFrames.y;\r\nfloat accumdist = 0;\r\nfloat3 curdensity = 0;\r\nfloat transmittance = 1;\r\nfloat3 localcamvec = normalize( mul(Parameters.CameraVector, (float3x3)GetPrimitiveData(Parameters).WorldToLocal) ) * StepSize;\r\n\r\nfloat3 invlightdir = 1 / LightVector;\r\n\r\nfloat shadowstepsize = 1 / ShadowSteps;\r\nLightVector *= shadowstepsize*0.25;\r\nShadowDensity *= shadowstepsize;\r\n\r\nDensity *= StepSize;\r\nfloat3 lightenergy = 0;\r\nfloat shadowthresh = -log(ShadowThreshold) / ShadowDensity;\r\n\r\nint3 randpos = int3(Parameters.SvPosition.xy, View.StateFrameIndexMod8);\r\nfloat rand =float(Rand3DPCG16(randpos).x) / 0xffff;\r\nrand=1;\r\nCurPos +=  localcamvec * rand.x * Jitter;\r\n\r\n\r\nfor (int i = 0; i < MaxSteps; i++)\r\n{\t\r\n\r\n\r\n\t##if EnableCurl\r\n\t\tfloat4 velocity = PseudoVolumeTexture(Tex2, TexSampler, saturate(CurPos), XYFrames, numFrames);\r\n\t\tfloat3 curlpos = CurPos * CurlTiling + CurlPan;\r\n\t\tcurlpos += normalize(velocity) * curladvect;\r\n\t\tfloat3 curl = PseudoVolumeTexture(CurlTex, TexSampler, curlpos, 16, 256).rgb;\r\n\t\tfloat pressure = PseudoVolumeTexture(Tex3, TexSampler, saturate(CurPos), XYFrames, numFrames).r;\r\n\t\tpressure = saturate( -pressure+PressureDisplay);\r\n\t\t\r\n\t\tcurl = CurlAmount * 2 * (curl - 0.5);\r\n\t\tcurl *= pressure;\r\n\t##else\r\n\t\tfloat3 curl = 0;\r\n\t##endif\r\n\r\n\tfloat3 samplepos =  saturate(CurPos + curl);\r\n\tsamplepos.z = min(samplepos.z, 1 - (1/VolumeResolution.z));\r\n\r\n\tfloat4 cursample = PseudoVolumeTexture(Tex, TexSampler, samplepos, XYFrames, numFrames);\r\n\t//float pressure = max(0, PseudoVolumeTexture(Tex3, TexSampler, saturate(CurPos), XYFrames, numFrames).r);\r\n\r\n\r\n\t//Sample Light Absorption and Scattering\r\n\tif( cursample.r + cursample.g + cursample.b > 0.001)\r\n\t{\r\n\t\tfloat3 lpos = CurPos - LightVector * LightOffset;\r\n\t\tfloat shadowdist = 0;\r\n\r\n\t\tfor (int s = 0; s < ShadowSteps; s++)\r\n\t\t{\r\n\t\t\tlpos += LightVector;\r\n\r\n\t\t\t##if EnableCurl\r\n\t\t\t\tvelocity = PseudoVolumeTexture(Tex2, TexSampler, saturate(lpos), XYFrames, numFrames);\r\n\t\t\t\tcurlpos = lpos * CurlTiling + CurlPan;\r\n\t\t\t\tcurlpos += normalize(velocity) * curladvect;\r\n\t\t\t\tcurl = PseudoVolumeTexture(CurlTex, TexSampler, curlpos , 16, 256).rgb;\r\n\t\t\t\tcurl = CurlAmount * 2 * (curl - 0.5);\r\n\r\n\t\t\t\tpressure = PseudoVolumeTexture(Tex3, TexSampler, saturate(CurPos), XYFrames, numFrames).r;\r\n\t\t\t\tpressure = saturate( -pressure+PressureDisplay);\r\n\t\t\t\tcurl *= pressure;\r\n\t\t\t##else\r\n\t\t\t\tcurl = 0;\r\n\t\t\t##endif\r\n\t\t\r\n\t\t\tfloat3 lsample = PseudoVolumeTexture(Tex, TexSampler, saturate(lpos + curl), XYFrames, numFrames).rgb;\r\n\t\t\t\r\n\t\t\tfloat3 shadowboxtest = floor( 0.5 + ( abs( 0.5 - lpos ) ) );\r\n\t\t\tfloat exitshadowbox = shadowboxtest .x + shadowboxtest .y + shadowboxtest .z;\r\n\r\n       \t\t\tif(shadowdist > shadowthresh || exitshadowbox >= 1) break;\r\n\r\n\t\t\tshadowdist += dot(lsample, 0.3333);\r\n\t\t}\r\n\t\r\n\t\tcurdensity = 1 - exp(-cursample.rgb * Density);\r\n\t\tlightenergy += exp(-shadowdist * ShadowDensity) * curdensity * transmittance * LightColor;\r\n\t\tlightenergy += (MaterialExpressionBlackBody(Temp*cursample.a)) * curdensity * transmittance;\r\n\t\t//lightenergy += pressure * transmittance * PressureDisplay;\r\n\r\n\t\t\r\n\t\ttransmittance *= 1-dot(curdensity, 0.333);\r\n\t\t##if 1\r\n\r\n\t\t//Sky Lighting\r\n\t\tshadowdist = 0;\r\n\t\t\r\n\t\tlpos = CurPos + float3(0,0,0.025);\r\n\t\tfloat3 lsample = PseudoVolumeTexture(Tex, TexSampler, saturate(lpos), XYFrames, numFrames).rgb;\r\n\t\tshadowdist += dot(lsample, 0.3333);\r\n\t\tlpos = CurPos + float3(0,0,0.05);\r\n\t\tlsample = PseudoVolumeTexture(Tex, TexSampler, saturate(lpos), XYFrames, numFrames).rgb;\r\n\t\tshadowdist += dot(lsample, 0.3333);\r\n\t\tlpos = CurPos + float3(0,0,0.15);\r\n\t\tlsample = PseudoVolumeTexture(Tex, TexSampler, saturate(lpos), XYFrames, numFrames).rgb;\r\n\t\tshadowdist += dot(lsample, 0.3333);\r\n\r\n\t\t//shadowterm    = exp(-shadowdist  * AmbientDensity);\r\n\t\t//absorbedlight = exp(-shadowdist  * AmbientDensity) * curdensity;\r\n\t\tlightenergy    += exp(-shadowdist  * AmbientDensity) * curdensity * SkyColor * transmittance;\r\n\t\t##endif\r\n\r\n\t}\r\n\t\t\r\n\r\n\tCurPos -= localcamvec;\r\n}\r\n\r\n\r\n\r\nreturn float4( lightenergy, transmittance);\r\n"
   description=Density RayMarch
   outputType=CMOT_Float4
   inputs [
    {
     inputName=Tex
     input {expression=&self:MaterialExpressionTextureObjectParameter_1}
    }
    {
     inputName=XYFrames
     input {expression=&self:MaterialExpressionComponentMask_4}
    }
    {
     inputName=StepSize
     input {expression=&self:MaterialExpressionDivide_8}
    }
    {
     inputName=MaxSteps
     input {expression=&self:MaterialExpressionClamp_0}
    }
    {
     inputName=FinalStep
     input {expression=&self:MaterialExpressionFrac_3}
    }
    {
     inputName=CurPos
     input {expression=&self:MaterialExpressionMultiply_0}
    }
    {
     inputName=Density
     input {expression=&self:MaterialExpressionScalarParameter_0}
    }
    {
     inputName=ShadowSteps
     input {expression=&self:MaterialExpressionScalarParameter_25}
    }
    {
     inputName=ShadowDensity
     input {expression=&self:MaterialExpressionScalarParameter_26}
    }
    {
     inputName=LightVector
     input {expression=&self:MaterialExpressionNormalize_1}
    }
    {
     inputName=ShadowThreshold
     input {expression=&self:MaterialExpressionScalarParameter_12}
    }
    {
     inputName=LightColor
     input {expression=&self:MaterialExpressionMultiply_17}
    }
    {
     inputName=SkyColor
     input {expression=&self:MaterialExpressionMultiply_19}
    }
    {
     inputName=AmbientDensity
     input {expression=&self:MaterialExpressionScalarParameter_4}
    }
    {
     inputName=Jitter
     input {expression=&self:MaterialExpressionMultiply_3}
    }
    {
     inputName=Tex2
     input {expression=&self:MaterialExpressionTextureObjectParameter_2}
    }
    {
     inputName=Temp
     input {expression=&self:MaterialExpressionScalarParameter_27}
    }
    {
     inputName=CurlTex
     input {expression=&self:MaterialExpressionTextureObject_0}
    }
    {
     inputName=CurlTiling
     input {expression=&self:MaterialExpressionScalarParameter_29}
    }
    {
     inputName=CurlAmount
     input {expression=&self:MaterialExpressionScalarParameter_30}
    }
    {
     inputName=CurlPan
     input {expression=&self:MaterialExpressionMultiply_20}
    }
    {
     inputName=Tex3
     input {expression=&self:MaterialExpressionTextureObjectParameter_3}
    }
    {
     inputName=PressureDisplay
     input {expression=&self:MaterialExpressionScalarParameter_7}
    }
    {
     inputName=curladvect
     input {expression=&self:MaterialExpressionScalarParameter_3}
    }
    {
     inputName=LightOffset
     input {expression=&self:MaterialExpressionScalarParameter_9}
    }
    {
     inputName=VolumeResolution
     input {
      expression=&self:MaterialExpressionVectorParameter_0
      mask=1
      maskB=1
      maskG=1
      maskR=1
     }
    }
   ]
  }
  Min(
   A=MaterialFunctionCall(X=@n3, Y=@n3, Z=@n3) {materialFunction=&MakeFloat3}.Result
   B=ComponentMask(None=@n8) {b=true, g=true, r=true}
  ) {constB=100}
  OneMinus(None=ComponentMask(None=@n8) {a=true})
 ]
}
