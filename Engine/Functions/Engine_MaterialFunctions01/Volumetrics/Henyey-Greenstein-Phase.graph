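# Material function that evaluates the Henyey-Greenstein phase function for anisotropic volumetric scattering, from isotropic at zero Anisotropy to forward scattering at one.
# The angle comes from a Custom-node acos of Light Vector dot camera vector, converted to degrees, unless AngleDegrees overrides it.
# It converts the angle back to radians for the cosine, then computes (1−g²) / (1+g²−2g·cosθ)^1.5, divided by 4π.
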MaterialFunction /Engine/Functions/Engine_MaterialFunctions01/Volumetrics/Henyey-Greenstein-Phase
graph {nodes [
  n1 = FunctionInput() {
   description=0 equals isotropic or equal scattering in all directions. 1 means full forward scattering.
   inputName=Anisotropy
   inputType=FunctionInput_Scalar
  }
  n2 = Multiply(A=@n1, B=@n1)
  n3 = OneMinus(None=@n2)
  n4 = Multiply(A=@n1) {constB=2}
  n5 = FunctionInput() {
   bUsePreviewValueAsDefault=true
   description="The light vector to calculate the viewing angle from. Note that if AngleDegrees is specified, this input will be ignored."
   inputName=Light Vector
   sortPriority=3
  }
  MaterialExpressionDotProduct_4 = DotProduct(A=@n5, B=CameraVectorWS())
  n6 = Custom!(x=@MaterialExpressionDotProduct_4) {
   code="acos(x);"
   description="acos(x)"
   outputType=CMOT_Float1
   inputs [{
     inputName=x
     input {expression=&self:MaterialExpressionDotProduct_4}
    }]
  }
  n7 = FunctionInput(Preview=Multiply(A=@n6) {constB=57.2958}) {
   bUsePreviewValueAsDefault=true
   description="By default this input is not used and the angle is calculated using the input Light Vector. If this input is specified, light vector will be disregarded and the input angle will be used. In Degrees."
   inputName=AngleDegrees
   inputType=FunctionInput_Scalar
   sortPriority=3
  }
  n8 = Cosine(None=Multiply(A=@n7) {constB=0.0175}) {period=6.2832}
  n9 = Power(Base=Subtract(A=Add(A=@n2), B=Multiply(A=@n4, B=@n8)), Exp=Divide() {constA=3})
  n10 = Constant() {r=4}
  FunctionOutput(None=Divide(
    A=Divide(A=@n3, B=@n9)
    B=MaterialFunctionCall(Multiplier=@n10) {materialFunction=&Pi}.Result
   ))
 ]}
