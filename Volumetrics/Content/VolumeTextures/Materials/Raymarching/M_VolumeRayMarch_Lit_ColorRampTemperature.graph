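# Unlit alpha-composite material that raymarches a self-lit, temperature-glowing volume, such as smoke or fire, from a flipbook pseudo-volume texture.
# VolumeBoxIntersect supplies ray entry and step count to a Custom HLSL loop, with temporal-dither jitter.
# The loop adds shadow-marched sunlight with Henyey-Greenstein phase, three-tap sky occlusion, and blackbody emission from alpha plus a density-band boost, returning light and transmittance.
# Velocity- and pressure-driven curl distortion is wired but compiled out.
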Material /Volumetrics/Content/VolumeTextures/Materials/Raymarching/M_VolumeRayMarch_Lit_ColorRampTemperature
graph {
 properties {bDisableDepthTest=true, blendMode=BLEND_AlphaComposite, pixelDepthOffsetMode=PDOM_Legacy, shadingModel=MSM_Unlit, usedShadingModels=Unlit}
 nodes [
  MaterialExpressionTextureObjectParameter_6 = TextureObjectParameter() {parameterName=Color, texture=/Volumetrics/Content/VolumeTextures/Textures/RT_Default}
  MaterialExpressionComponentMask_8 = ComponentMask(None=VectorParameter() {parameterName=XYFrames, defaultValue=<rgba 10 10 0 0>}.RGB) {g=true, r=true}
  MaterialExpressionDivide_0 = Divide(B=ScalarParameter() {defaultValue=128, parameterName=Steps})
  n1 = MaterialFunctionCall!(
   Steps=ScalarParameter() {defaultValue=128, parameterName=Steps}
   Plane alignment.=ScalarParameter() {defaultValue=1, parameterName=Plane Alignment}
  ) {materialFunction=/Volumetrics/Content/VolumeTextures/MaterialFunctions/VolumeBoxIntersect}
  n2 = Multiply(A=@n1.Box Distance, B=ScalarParameter() {defaultValue=128, parameterName=Steps})
  MaterialExpressionClamp_1 = Clamp(None=Floor(None=@n2)) {desc=MaxSteps, maxDefault=512}
  MaterialExpressionFrac_0 = Frac(None=@n2) {desc=FinalStepSize}
  MaterialExpressionMultiply_2 = Multiply(A=@n1.Intersection Position) {desc=Ray Entry}
  MaterialExpressionScalarParameter_18 = ScalarParameter() {defaultValue=16, parameterName=Density}
  MaterialExpressionScalarParameter_16 = ScalarParameter() {defaultValue=16, parameterName=ShadowSteps}
  MaterialExpressionScalarParameter_17 = ScalarParameter() {defaultValue=4, parameterName=ShadowDensity}
  MaterialExpressionNormalize_0 = Normalize(VectorInput=Transform(None=VectorParameter() {parameterName=LightVector, defaultValue=<rgba 1 0 1 1>}.RGB) {transformSourceType=TRANSFORMSOURCE_World, transformType=TRANSFORM_Local})
  MaterialExpressionScalarParameter_19 = ScalarParameter() {defaultValue=0.01, parameterName=SB}
  n3 = VectorParameter() {parameterName=Sun Color, defaultValue=<rgba 1 0.875 0.55 4>}
  n4 = ScalarParameter() {defaultValue=0.3, parameterName=Aniso}
  n5 = Multiply(A=Normalize(VectorInput=VectorParameter() {parameterName=LightVector, defaultValue=<rgba 1 0 1 1>}.RGB)) {constB=-1}
  MaterialExpressionMultiply_5 = Multiply(
   A=@n3.RGB
   B=Multiply(
    A=@n3.A
    B=MaterialFunctionCall!(Anisotropy=@n4, Light Vector=@n5, AngleDegrees) {materialFunction=/Engine/Functions/Engine_MaterialFunctions01/Volumetrics/Henyey-Greenstein-Phase}.Result
   )
  )
  n6 = VectorParameter() {parameterName=Sky Color, defaultValue=<rgba 0.2 0.2 0.25 0.2>}
  MaterialExpressionMultiply_6 = Multiply(A=@n6.RGB, B=@n6.RGB)
  MaterialExpressionScalarParameter_20 = ScalarParameter() {defaultValue=0.2, parameterName=Ambient Density}
  MaterialExpressionMultiply_9 = Multiply(
   A=MaterialFunctionCall(Alpha Threshold=ScalarParameter() {parameterName=wtf}) {materialFunction=&DitherTemporalAA}.Result
   B=ScalarParameter() {defaultValue=1, parameterName=Jitter}
  )
  MaterialExpressionTextureObjectParameter_4 = TextureObjectParameter() {parameterName=RT_Velocity, samplerType=SAMPLERTYPE_LinearColor, texture=&RT_Default}
  MaterialExpressionScalarParameter_31 = ScalarParameter() {defaultValue=4000, parameterName=Temp}
  MaterialExpressionTextureObject_1 = TextureObject() {samplerType=SAMPLERTYPE_LinearColor, texture=/Volumetrics/Content/VolumeTextures/Textures/T_CurlNoise_Volume_Tiling_01}
  MaterialExpressionScalarParameter_32 = ScalarParameter() {defaultValue=0.4, parameterName=Curl Tiling}
  MaterialExpressionScalarParameter_34 = ScalarParameter() {defaultValue=0.03, parameterName=Curl Amount}
  MaterialExpressionMultiply_10 = Multiply(
   A=VectorParameter() {parameterName=Panner, defaultValue=<rgba 0 1 -1 0>}.RGB
   B=MaterialFunctionCall(Speed=ScalarParameter() {parameterName=Speed}) {materialFunction=&TimeWithSpeedVariable}.Time
  )
  MaterialExpressionTextureObjectParameter_5 = TextureObjectParameter() {parameterName=Pressure, samplerType=SAMPLERTYPE_LinearColor, texture=&RT_Default}
  MaterialExpressionScalarParameter_38 = ScalarParameter() {defaultValue=0.5, parameterName=fasdfasdfasdf}
  MaterialExpressionConstant_2 = Constant() {r=1}
  MaterialExpressionScalarParameter_37 = ScalarParameter() {defaultValue=0.025, parameterName=CurlAdvection}
  MaterialExpressionScalarParameter_39 = ScalarParameter() {defaultValue=0.75, parameterName=lo}
  MaterialExpressionVectorParameter_14 = VectorParameter() {parameterName=VolumeResolution, defaultValue=<rgba 256 256 256 0>}
  MaterialExpressionScalarParameter_40 = ScalarParameter() {defaultValue=0.3, parameterName=Temperature Loc}
  MaterialExpressionScalarParameter_41 = ScalarParameter() {defaultValue=0.5, parameterName=Temperature Width}
  MaterialExpressionScalarParameter_42 = ScalarParameter() {defaultValue=4, parameterName=Temperature Boost}
  MaterialExpressionScalarParameter_43 = ScalarParameter() {defaultValue=0.6, parameterName=EdgeTemp}
  n7 = Custom!(
   Tex=@MaterialExpressionTextureObjectParameter_6
   XYFrames=@MaterialExpressionComponentMask_8
   StepSize=@MaterialExpressionDivide_0
   MaxSteps=@MaterialExpressionClamp_1
   FinalStep=@MaterialExpressionFrac_0
   CurPos=@MaterialExpressionMultiply_2
   Density=@MaterialExpressionScalarParameter_18
   ShadowSteps=@MaterialExpressionScalarParameter_16
   ShadowDensity=@MaterialExpressionScalarParameter_17
   LightVector=@MaterialExpressionNormalize_0
   ShadowThreshold=@MaterialExpressionScalarParameter_19
   LightColor=@MaterialExpressionMultiply_5
   SkyColor=@MaterialExpressionMultiply_6
   AmbientDensity=@MaterialExpressionScalarParameter_20
   Jitter=@MaterialExpressionMultiply_9
   Tex2=@MaterialExpressionTextureObjectParameter_4
   Temp=@MaterialExpressionScalarParameter_31
   CurlTex=@MaterialExpressionTextureObject_1
   CurlTiling=@MaterialExpressionScalarParameter_32
   CurlAmount=@MaterialExpressionScalarParameter_34
   CurlPan=@MaterialExpressionMultiply_10
   Tex3=@MaterialExpressionTextureObjectParameter_5
   PressureDisplay=@MaterialExpressionScalarParameter_38
   UseCurlNoise=@MaterialExpressionConstant_2
   curladvect=@MaterialExpressionScalarParameter_37
   LightOffset=@MaterialExpressionScalarParameter_39
   VolumeResolution=@MaterialExpressionVectorParameter_14.RGB
   TempLoc=@MaterialExpressionScalarParameter_40
   TempWidth=@MaterialExpressionScalarParameter_41
   TempBoost=@MaterialExpressionScalarParameter_42
   EdgeTemp=@MaterialExpressionScalarParameter_43
  ) {
   code="##define EnableCurl 0\r\nfloat numFrames = XYFrames.x * XYFrames.y;\r\nfloat accumdist = 0;\r\nfloat3 curdensity = 0;\r\nfloat transmittance = 1;\r\nfloat3 localcamvec = normalize( mul(Parameters.CameraVector, (float3x3)WSDemote(GetWorldToLocal(Parameters))) ) * StepSize;\r\n\r\nfloat3 invlightdir = 1 / LightVector;\r\n\r\nfloat shadowstepsize = 1 / ShadowSteps;\r\nLightVector *= shadowstepsize*0.25;\r\nShadowDensity *= shadowstepsize;\r\n\r\nDensity *= StepSize;\r\nfloat3 lightenergy = 0;\r\nfloat shadowthresh = -log(ShadowThreshold) / ShadowDensity;\r\n\r\nint3 randpos = int3(Parameters.SvPosition.xy, View.StateFrameIndexMod8);\r\nfloat rand =float(Rand3DPCG16(randpos).x) / 0xffff;\r\nrand=1;\r\nCurPos +=  localcamvec * rand.x * Jitter;\r\n\r\n\r\nfor (int i = 0; i < MaxSteps; i++)\r\n{\t\r\n\r\n\r\n\t##if EnableCurl\r\n\t\tfloat4 velocity = PseudoVolumeTexture(Tex2, TexSampler, saturate(CurPos), XYFrames, numFrames);\r\n\t\tfloat3 curlpos = CurPos * CurlTiling + CurlPan;\r\n\t\tcurlpos += normalize(velocity) * curladvect;\r\n\t\tfloat3 curl = PseudoVolumeTexture(CurlTex, TexSampler, curlpos, 16, 256).rgb;\r\n\t\tfloat pressure = PseudoVolumeTexture(Tex3, TexSampler, saturate(CurPos), XYFrames, numFrames).r;\r\n\t\tpressure = saturate( -pressure+PressureDisplay);\r\n\t\t\r\n\t\tcurl = CurlAmount * 2 * (curl - 0.5);\r\n\t\tcurl *= pressure;\r\n\t##else\r\n\t\tfloat3 curl = 0;\r\n\t##endif\r\n\r\n\tfloat3 samplepos =  saturate(CurPos + curl);\r\n\tsamplepos.z = min(samplepos.z, 1 - (1/VolumeResolution.z));\r\n\r\n\tfloat4 cursample = PseudoVolumeTexture(Tex, TexSampler, samplepos, XYFrames, numFrames);\r\n\t//float pressure = max(0, PseudoVolumeTexture(Tex3, TexSampler, saturate(CurPos), XYFrames, numFrames).r);\r\n\r\n\r\n\t//Sample Light Absorption and Scattering\r\n\tif( cursample.r + cursample.g + cursample.b > 0.001)\r\n\t{\r\n\t\tfloat3 lpos = CurPos - LightVector * LightOffset;\r\n\t\tfloat shadowdist = 0;\r\n\r\n\t\tfor (int s = 0; s < ShadowSteps; s++)\r\n\t\t{\r\n\t\t\tlpos += LightVector;\r\n\r\n\t\t\t##if EnableCurl\r\n\t\t\t\tvelocity = PseudoVolumeTexture(Tex2, TexSampler, saturate(lpos), XYFrames, numFrames);\r\n\t\t\t\tcurlpos = lpos * CurlTiling + CurlPan;\r\n\t\t\t\tcurlpos += normalize(velocity) * curladvect;\r\n\t\t\t\tcurl = PseudoVolumeTexture(CurlTex, TexSampler, curlpos , 16, 256).rgb;\r\n\t\t\t\tcurl = CurlAmount * 2 * (curl - 0.5);\r\n\r\n\t\t\t\tpressure = PseudoVolumeTexture(Tex3, TexSampler, saturate(CurPos), XYFrames, numFrames).r;\r\n\t\t\t\tpressure = saturate( -pressure+PressureDisplay);\r\n\t\t\t\tcurl *= pressure;\r\n\t\t\t##else\r\n\t\t\t\tcurl = 0;\r\n\t\t\t##endif\r\n\t\t\r\n\t\t\tfloat3 lsample = PseudoVolumeTexture(Tex, TexSampler, saturate(lpos + curl), XYFrames, numFrames).rgb;\r\n\t\t\t\r\n\t\t\tfloat3 shadowboxtest = floor( 0.5 + ( abs( 0.5 - lpos ) ) );\r\n\t\t\tfloat exitshadowbox = shadowboxtest .x + shadowboxtest .y + shadowboxtest .z;\r\n\r\n       \t\t\tif(shadowdist > shadowthresh || exitshadowbox >= 1) break;\r\n\r\n\t\t\tshadowdist += dot(lsample, 0.3333);\r\n\t\t}\r\n\t\r\n\t\tcurdensity = 1 - exp(-cursample.rgb * Density);\r\n\t\tlightenergy += exp(-shadowdist * ShadowDensity) * curdensity * transmittance * LightColor;\r\n\r\n\t\tfloat fresneltemp = (1 - saturate(abs(cursample.r - TempLoc) / TempWidth)) * EdgeTemp;\r\n\t\tlightenergy += (MaterialExpressionBlackBody(Temp*(cursample.a + fresneltemp))) * curdensity * transmittance * TempBoost;\r\n\t\t//lightenergy += pressure * transmittance * PressureDisplay;\r\n\r\n\t\t\r\n\t\ttransmittance *= 1-dot(curdensity, 0.333);\r\n\t\t##if 1\r\n\r\n\t\t//Sky Lighting\r\n\t\tshadowdist = 0;\r\n\t\t\r\n\t\tlpos = CurPos + float3(0,0,0.025);\r\n\t\tfloat3 lsample = PseudoVolumeTexture(Tex, TexSampler, saturate(lpos), XYFrames, numFrames).rgb;\r\n\t\tshadowdist += dot(lsample, 0.3333);\r\n\t\tlpos = CurPos + float3(0,0,0.05);\r\n\t\tlsample = PseudoVolumeTexture(Tex, TexSampler, saturate(lpos), XYFrames, numFrames).rgb;\r\n\t\tshadowdist += dot(lsample, 0.3333);\r\n\t\tlpos = CurPos + float3(0,0,0.15);\r\n\t\tlsample = PseudoVolumeTexture(Tex, TexSampler, saturate(lpos), XYFrames, numFrames).rgb;\r\n\t\tshadowdist += dot(lsample, 0.3333);\r\n\r\n\t\t//shadowterm    = exp(-shadowdist  * AmbientDensity);\r\n\t\t//absorbedlight = exp(-shadowdist  * AmbientDensity) * curdensity;\r\n\t\tlightenergy    += exp(-shadowdist  * AmbientDensity) * curdensity * SkyColor * transmittance;\r\n\t\t##endif\r\n\r\n\t}\r\n\t\t\r\n\r\n\tCurPos -= localcamvec;\r\n}\r\n\r\n\r\n\r\nreturn float4( lightenergy, transmittance);\r\n"
   desc="\n*** Original source before expression upgrade ***\n##define EnableCurl 0\r\nfloat numFrames = XYFrames.x * XYFrames.y;\r\nfloat accumdist = 0;\r\nfloat3 curdensity = 0;\r\nfloat transmittance = 1;\r\nfloat3 localcamvec = normalize( mul(Parameters.CameraVector, (float3x3)GetPrimitiveData(Parameters).WorldToLocal) ) * StepSize;\r\n\r\nfloat3 invlightdir = 1 / LightVector;\r\n\r\nfloat shadowstepsize = 1 / ShadowSteps;\r\nLightVector *= shadowstepsize*0.25;\r\nShadowDensity *= shadowstepsize;\r\n\r\nDensity *= StepSize;\r\nfloat3 lightenergy = 0;\r\nfloat shadowthresh = -log(ShadowThreshold) / ShadowDensity;\r\n\r\nint3 randpos = int3(Parameters.SvPosition.xy, View.StateFrameIndexMod8);\r\nfloat rand =float(Rand3DPCG16(randpos).x) / 0xffff;\r\nrand=1;\r\nCurPos +=  localcamvec * rand.x * Jitter;\r\n\r\n\r\nfor (int i = 0; i < MaxSteps; i++)\r\n{\t\r\n\r\n\r\n\t##if EnableCurl\r\n\t\tfloat4 velocity = PseudoVolumeTexture(Tex2, TexSampler, saturate(CurPos), XYFrames, numFrames);\r\n\t\tfloat3 curlpos = CurPos * CurlTiling + CurlPan;\r\n\t\tcurlpos += normalize(velocity) * curladvect;\r\n\t\tfloat3 curl = PseudoVolumeTexture(CurlTex, TexSampler, curlpos, 16, 256).rgb;\r\n\t\tfloat pressure = PseudoVolumeTexture(Tex3, TexSampler, saturate(CurPos), XYFrames, numFrames).r;\r\n\t\tpressure = saturate( -pressure+PressureDisplay);\r\n\t\t\r\n\t\tcurl = CurlAmount * 2 * (curl - 0.5);\r\n\t\tcurl *= pressure;\r\n\t##else\r\n\t\tfloat3 curl = 0;\r\n\t##endif\r\n\r\n\tfloat3 samplepos =  saturate(CurPos + curl);\r\n\tsamplepos.z = min(samplepos.z, 1 - (1/VolumeResolution.z));\r\n\r\n\tfloat4 cursample = PseudoVolumeTexture(Tex, TexSampler, samplepos, XYFrames, numFrames);\r\n\t//float pressure = max(0, PseudoVolumeTexture(Tex3, TexSampler, saturate(CurPos), XYFrames, numFrames).r);\r\n\r\n\r\n\t//Sample Light Absorption and Scattering\r\n\tif( cursample.r + cursample.g + cursample.b > 0.001)\r\n\t{\r\n\t\tfloat3 lpos = CurPos - LightVector * LightOffset;\r\n\t\tfloat shadowdist = 0;\r\n\r\n\t\tfor (int s = 0; s < ShadowSteps; s++)\r\n\t\t{\r\n\t\t\tlpos += LightVector;\r\n\r\n\t\t\t##if EnableCurl\r\n\t\t\t\tvelocity = PseudoVolumeTexture(Tex2, TexSampler, saturate(lpos), XYFrames, numFrames);\r\n\t\t\t\tcurlpos = lpos * CurlTiling + CurlPan;\r\n\t\t\t\tcurlpos += normalize(velocity) * curladvect;\r\n\t\t\t\tcurl = PseudoVolumeTexture(CurlTex, TexSampler, curlpos , 16, 256).rgb;\r\n\t\t\t\tcurl = CurlAmount * 2 * (curl - 0.5);\r\n\r\n\t\t\t\tpressure = PseudoVolumeTexture(Tex3, TexSampler, saturate(CurPos), XYFrames, numFrames).r;\r\n\t\t\t\tpressure = saturate( -pressure+PressureDisplay);\r\n\t\t\t\tcurl *= pressure;\r\n\t\t\t##else\r\n\t\t\t\tcurl = 0;\r\n\t\t\t##endif\r\n\t\t\r\n\t\t\tfloat3 lsample = PseudoVolumeTexture(Tex, TexSampler, saturate(lpos + curl), XYFrames, numFrames).rgb;\r\n\t\t\t\r\n\t\t\tfloat3 shadowboxtest = floor( 0.5 + ( abs( 0.5 - lpos ) ) );\r\n\t\t\tfloat exitshadowbox = shadowboxtest .x + shadowboxtest .y + shadowboxtest .z;\r\n\r\n       \t\t\tif(shadowdist > shadowthresh || exitshadowbox >= 1) break;\r\n\r\n\t\t\tshadowdist += dot(lsample, 0.3333);\r\n\t\t}\r\n\t\r\n\t\tcurdensity = 1 - exp(-cursample.rgb * Density);\r\n\t\tlightenergy += exp(-shadowdist * ShadowDensity) * curdensity * transmittance * LightColor;\r\n\r\n\t\tfloat fresneltemp = (1 - saturate(abs(cursample.r - TempLoc) / TempWidth)) * EdgeTemp;\r\n\t\tlightenergy += (MaterialExpressionBlackBody(Temp*(cursample.a + fresneltemp))) * curdensity * transmittance * TempBoost;\r\n\t\t//lightenergy += pressure * transmittance * PressureDisplay;\r\n\r\n\t\t\r\n\t\ttransmittance *= 1-dot(curdensity, 0.333);\r\n\t\t##if 1\r\n\r\n\t\t//Sky Lighting\r\n\t\tshadowdist = 0;\r\n\t\t\r\n\t\tlpos = CurPos + float3(0,0,0.025);\r\n\t\tfloat3 lsample = PseudoVolumeTexture(Tex, TexSampler, saturate(lpos), XYFrames, numFrames).rgb;\r\n\t\tshadowdist += dot(lsample, 0.3333);\r\n\t\tlpos = CurPos + float3(0,0,0.05);\r\n\t\tlsample = PseudoVolumeTexture(Tex, TexSampler, saturate(lpos), XYFrames, numFrames).rgb;\r\n\t\tshadowdist += dot(lsample, 0.3333);\r\n\t\tlpos = CurPos + float3(0,0,0.15);\r\n\t\tlsample = PseudoVolumeTexture(Tex, TexSampler, saturate(lpos), XYFrames, numFrames).rgb;\r\n\t\tshadowdist += dot(lsample, 0.3333);\r\n\r\n\t\t//shadowterm    = exp(-shadowdist  * AmbientDensity);\r\n\t\t//absorbedlight = exp(-shadowdist  * AmbientDensity) * curdensity;\r\n\t\tlightenergy    += exp(-shadowdist  * AmbientDensity) * curdensity * SkyColor * transmittance;\r\n\t\t##endif\r\n\r\n\t}\r\n\t\t\r\n\r\n\tCurPos -= localcamvec;\r\n}\r\n\r\n\r\n\r\nreturn float4( lightenergy, transmittance);\r\n"
   description=Density RayMarch
   outputType=CMOT_Float4
   inputs [
    {
     inputName=Tex
     input {expression=&self:MaterialExpressionTextureObjectParameter_6}
    }
    {
     inputName=XYFrames
     input {expression=&self:MaterialExpressionComponentMask_8}
    }
    {
     inputName=StepSize
     input {expression=&self:MaterialExpressionDivide_0}
    }
    {
     inputName=MaxSteps
     input {expression=&self:MaterialExpressionClamp_1}
    }
    {
     inputName=FinalStep
     input {expression=&self:MaterialExpressionFrac_0}
    }
    {
     inputName=CurPos
     input {expression=&self:MaterialExpressionMultiply_2}
    }
    {
     inputName=Density
     input {expression=&self:MaterialExpressionScalarParameter_18}
    }
    {
     inputName=ShadowSteps
     input {expression=&self:MaterialExpressionScalarParameter_16}
    }
    {
     inputName=ShadowDensity
     input {expression=&self:MaterialExpressionScalarParameter_17}
    }
    {
     inputName=LightVector
     input {expression=&self:MaterialExpressionNormalize_0}
    }
    {
     inputName=ShadowThreshold
     input {expression=&self:MaterialExpressionScalarParameter_19}
    }
    {
     inputName=LightColor
     input {expression=&self:MaterialExpressionMultiply_5}
    }
    {
     inputName=SkyColor
     input {expression=&self:MaterialExpressionMultiply_6}
    }
    {
     inputName=AmbientDensity
     input {expression=&self:MaterialExpressionScalarParameter_20}
    }
    {
     inputName=Jitter
     input {expression=&self:MaterialExpressionMultiply_9}
    }
    {
     inputName=Tex2
     input {expression=&self:MaterialExpressionTextureObjectParameter_4}
    }
    {
     inputName=Temp
     input {expression=&self:MaterialExpressionScalarParameter_31}
    }
    {
     inputName=CurlTex
     input {expression=&self:MaterialExpressionTextureObject_1}
    }
    {
     inputName=CurlTiling
     input {expression=&self:MaterialExpressionScalarParameter_32}
    }
    {
     inputName=CurlAmount
     input {expression=&self:MaterialExpressionScalarParameter_34}
    }
    {
     inputName=CurlPan
     input {expression=&self:MaterialExpressionMultiply_10}
    }
    {
     inputName=Tex3
     input {expression=&self:MaterialExpressionTextureObjectParameter_5}
    }
    {
     inputName=PressureDisplay
     input {expression=&self:MaterialExpressionScalarParameter_38}
    }
    {
     inputName=UseCurlNoise
     input {expression=&self:MaterialExpressionConstant_2}
    }
    {
     inputName=curladvect
     input {expression=&self:MaterialExpressionScalarParameter_37}
    }
    {
     inputName=LightOffset
     input {expression=&self:MaterialExpressionScalarParameter_39}
    }
    {
     inputName=VolumeResolution
     input {
      expression=&self:MaterialExpressionVectorParameter_14
      mask=1
      maskB=1
      maskG=1
      maskR=1
     }
    }
    {
     inputName=TempLoc
     input {expression=&self:MaterialExpressionScalarParameter_40}
    }
    {
     inputName=TempWidth
     input {expression=&self:MaterialExpressionScalarParameter_41}
    }
    {
     inputName=TempBoost
     input {expression=&self:MaterialExpressionScalarParameter_42}
    }
    {
     inputName=EdgeTemp
     input {expression=&self:MaterialExpressionScalarParameter_43}
    }
   ]
  }
  ComponentMask(None=@n7) {b=true, g=true, r=true}
  OneMinus(None=ComponentMask(None=@n7) {a=true})
 ]
}
