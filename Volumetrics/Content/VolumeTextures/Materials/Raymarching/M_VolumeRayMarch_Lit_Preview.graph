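# Unlit alpha-composite preview material that raymarches a lit volume from a flipbook-style pseudo-volume texture inside a box, with depth testing disabled.
# VolumeBoxIntersect supplies ray entry and step count to a custom HLSL loop, which accumulates light and transmittance.
# Dense samples march toward LightVector for early-exit sun shadows and take three upward samples for sky ambient.
# Optional per-pixel, per-frame jitter offsets ray starts; the FinalStep input is unused.
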Material /Volumetrics/Content/VolumeTextures/Materials/Raymarching/M_VolumeRayMarch_Lit_Preview
graph {
 properties {bDisableDepthTest=true, blendMode=BLEND_AlphaComposite, pixelDepthOffsetMode=PDOM_Legacy, shadingModel=MSM_Unlit, usedShadingModels=Unlit}
 nodes [
  MaterialExpressionTextureObjectParameter_1 = TextureObjectParameter() {parameterName=RT, samplerType=SAMPLERTYPE_LinearColor, texture=/Volumetrics/Content/VolumeTextures/Textures/T_Lightning}
  MaterialExpressionComponentMask_2 = ComponentMask(None=VectorParameter() {parameterName=XYFrames, defaultValue=<rgba 8 8 0 0>}.RGB) {g=true, r=true}
  MaterialExpressionDivide_8 = Divide(B=ScalarParameter() {defaultValue=256, parameterName=Steps})
  n1 = MaterialFunctionCall!(
   Steps=ScalarParameter() {defaultValue=256, parameterName=Steps}
   Plane alignment.=ScalarParameter() {defaultValue=1, parameterName=Plane Alignment}
  ) {materialFunction=/Volumetrics/Content/VolumeTextures/MaterialFunctions/VolumeBoxIntersect}
  n2 = Multiply(A=@n1.Box Distance, B=ScalarParameter() {defaultValue=256, parameterName=Steps})
  MaterialExpressionClamp_0 = Clamp(None=Floor(None=@n2)) {desc=MaxSteps, maxDefault=512}
  MaterialExpressionFrac_3 = Frac(None=@n2) {desc=FinalStepSize}
  MaterialExpressionMultiply_0 = Multiply(A=@n1.Intersection Position) {desc=Ray Entry}
  MaterialExpressionScalarParameter_0 = ScalarParameter() {defaultValue=64, parameterName=Density}
  MaterialExpressionScalarParameter_25 = ScalarParameter() {defaultValue=32, parameterName=ShadowSteps}
  MaterialExpressionScalarParameter_26 = ScalarParameter() {defaultValue=64, parameterName=ShadowDensity}
  MaterialExpressionNormalize_1 = Normalize(VectorInput=Transform(None=VectorParameter() {parameterName=LightVector, defaultValue=<rgba 0 -1 5 1>}.RGB) {transformSourceType=TRANSFORMSOURCE_World, transformType=TRANSFORM_Local})
  MaterialExpressionScalarParameter_12 = ScalarParameter() {defaultValue=0.001, parameterName=SB}
  n3 = VectorParameter() {parameterName=Sun Color, defaultValue=<rgba 1 0.875 0.55 1>}
  MaterialExpressionMultiply_17 = Multiply(A=@n3.RGB, B=@n3.RGB)
  n4 = VectorParameter() {parameterName=Sky Color, defaultValue=<rgba 0.2 0.2 0.25 0.5>}
  MaterialExpressionMultiply_19 = Multiply(A=@n4.RGB, B=@n4.RGB)
  MaterialExpressionScalarParameter_4 = ScalarParameter() {defaultValue=1, parameterName=Ambient Density}
  MaterialExpressionScalarParameter_13 = ScalarParameter() {parameterName=Jitter}
  n5 = Custom!(
   Tex=@MaterialExpressionTextureObjectParameter_1
   XYFrames=@MaterialExpressionComponentMask_2
   StepSize=@MaterialExpressionDivide_8
   MaxSteps=@MaterialExpressionClamp_0
   FinalStep=@MaterialExpressionFrac_3
   CurPos=@MaterialExpressionMultiply_0
   Density=@MaterialExpressionScalarParameter_0
   ShadowSteps=@MaterialExpressionScalarParameter_25
   ShadowDensity=@MaterialExpressionScalarParameter_26
   LightVector=@MaterialExpressionNormalize_1
   ShadowThreshold=@MaterialExpressionScalarParameter_12
   LightColor=@MaterialExpressionMultiply_17
   SkyColor=@MaterialExpressionMultiply_19
   AmbientDensity=@MaterialExpressionScalarParameter_4
   Jitter=@MaterialExpressionScalarParameter_13
  ) {
   code="float numFrames = XYFrames.x * XYFrames.y;\r\nfloat accumdist = 0;\r\nfloat curdensity = 0;\r\nfloat transmittance = 1;\r\nfloat3 localcamvec = normalize( mul(Parameters.CameraVector, (float3x3)WSDemote(GetWorldToLocal(Parameters))) ) * StepSize;\r\n\r\nfloat3 invlightdir = 1 / LightVector;\r\n\r\nfloat shadowstepsize = 1 / ShadowSteps;\r\nLightVector *= shadowstepsize*0.5;\r\nShadowDensity *= shadowstepsize;\r\n\r\nDensity *= StepSize;\r\nfloat3 lightenergy = 0;\r\nfloat shadowthresh = -log(ShadowThreshold) / ShadowDensity;\r\n\r\nint3 randpos = int3(Parameters.SvPosition.xy, View.StateFrameIndexMod8);\r\nfloat rand =float(Rand3DPCG16(randpos).x) / 0xffff;\r\nCurPos +=  localcamvec * rand.x * Jitter;\r\n\r\n\r\nfor (int i = 0; i < MaxSteps; i++)\r\n{\t\r\n\r\n\t\r\n\tfloat cursample = PseudoVolumeTexture(Tex, TexSampler, CurPos, XYFrames, numFrames).r;\r\n\t\r\n\t//Sample Light Absorption and Scattering\r\n\tif( (cursample.r) > 0.001)\r\n\t{\r\n\t\tfloat3 lpos = CurPos;\r\n\t\tfloat shadowdist = 0;\r\n\r\n\t\tfor (int s = 0; s < ShadowSteps; s++)\r\n\t\t{\r\n\t\t\tlpos += LightVector;\r\n\r\n\t\t\r\n\r\n\t\t\tfloat lsample = PseudoVolumeTexture(Tex, TexSampler, saturate(lpos), XYFrames, numFrames).r;\r\n\t\t\t\r\n\t\t\tfloat3 shadowboxtest = floor( 0.5 + ( abs( 0.5 - lpos ) ) );\r\n\t\t\tfloat exitshadowbox = shadowboxtest .x + shadowboxtest .y + shadowboxtest .z;\r\n\r\n       \t\t\tif(shadowdist > shadowthresh || exitshadowbox >= 1) break;\r\n\r\n\t\t\tshadowdist += lsample;\r\n\t\t}\r\n\t\r\n\t\tcurdensity = 1 - exp(-cursample.r * Density);\r\n\r\n\r\n\t\t//curdensity = saturate(cursample * Density);\r\n\t\t//float     shadowterm = exp(-shadowdist * ShadowDensity);\r\n\t\t//float3 absorbedlight = exp(-shadowdist * ShadowDensity) * curdensity;\r\n\r\n\t\tlightenergy += exp(-shadowdist * ShadowDensity) * curdensity * transmittance * LightColor;\r\n\r\n\t\ttransmittance *= 1- (curdensity);\r\n\t\t\r\n\t\t##if 1\r\n\r\n\t\t//Sky Lighting\r\n\t\tshadowdist = 0;\r\n\t\t\r\n\t\tlpos = CurPos + float3(0,0,0.025);\r\n\t\tfloat lsample = PseudoVolumeTexture(Tex, TexSampler, saturate(lpos), XYFrames, numFrames).r;\r\n\t\tshadowdist += lsample;\r\n\t\tlpos = CurPos + float3(0,0,0.05);\r\n\t\tlsample = PseudoVolumeTexture(Tex, TexSampler, saturate(lpos), XYFrames, numFrames).r;\r\n\t\tshadowdist += lsample;\r\n\t\tlpos = CurPos + float3(0,0,0.15);\r\n\t\tlsample = PseudoVolumeTexture(Tex, TexSampler, saturate(lpos), XYFrames, numFrames).r;\r\n\t\tshadowdist += lsample;\r\n\r\n\t\tlightenergy    += exp(-shadowdist  * AmbientDensity) * curdensity * SkyColor * transmittance;\r\n\t\t##endif\r\n\r\n\t}\r\n\t\t\r\n\r\n\tCurPos -= localcamvec;\r\n}\r\n\r\n\r\n\r\nreturn float4( lightenergy, transmittance);\r\n"
   desc="\n*** Original source before expression upgrade ***\nfloat numFrames = XYFrames.x * XYFrames.y;\r\nfloat accumdist = 0;\r\nfloat curdensity = 0;\r\nfloat transmittance = 1;\r\nfloat3 localcamvec = normalize( mul(Parameters.CameraVector, (float3x3)GetPrimitiveData(Parameters).WorldToLocal) ) * StepSize;\r\n\r\nfloat3 invlightdir = 1 / LightVector;\r\n\r\nfloat shadowstepsize = 1 / ShadowSteps;\r\nLightVector *= shadowstepsize*0.5;\r\nShadowDensity *= shadowstepsize;\r\n\r\nDensity *= StepSize;\r\nfloat3 lightenergy = 0;\r\nfloat shadowthresh = -log(ShadowThreshold) / ShadowDensity;\r\n\r\nint3 randpos = int3(Parameters.SvPosition.xy, View.StateFrameIndexMod8);\r\nfloat rand =float(Rand3DPCG16(randpos).x) / 0xffff;\r\nCurPos +=  localcamvec * rand.x * Jitter;\r\n\r\n\r\nfor (int i = 0; i < MaxSteps; i++)\r\n{\t\r\n\r\n\t\r\n\tfloat cursample = PseudoVolumeTexture(Tex, TexSampler, CurPos, XYFrames, numFrames).r;\r\n\t\r\n\t//Sample Light Absorption and Scattering\r\n\tif( (cursample.r) > 0.001)\r\n\t{\r\n\t\tfloat3 lpos = CurPos;\r\n\t\tfloat shadowdist = 0;\r\n\r\n\t\tfor (int s = 0; s < ShadowSteps; s++)\r\n\t\t{\r\n\t\t\tlpos += LightVector;\r\n\r\n\t\t\r\n\r\n\t\t\tfloat lsample = PseudoVolumeTexture(Tex, TexSampler, saturate(lpos), XYFrames, numFrames).r;\r\n\t\t\t\r\n\t\t\tfloat3 shadowboxtest = floor( 0.5 + ( abs( 0.5 - lpos ) ) );\r\n\t\t\tfloat exitshadowbox = shadowboxtest .x + shadowboxtest .y + shadowboxtest .z;\r\n\r\n       \t\t\tif(shadowdist > shadowthresh || exitshadowbox >= 1) break;\r\n\r\n\t\t\tshadowdist += lsample;\r\n\t\t}\r\n\t\r\n\t\tcurdensity = 1 - exp(-cursample.r * Density);\r\n\r\n\r\n\t\t//curdensity = saturate(cursample * Density);\r\n\t\t//float     shadowterm = exp(-shadowdist * ShadowDensity);\r\n\t\t//float3 absorbedlight = exp(-shadowdist * ShadowDensity) * curdensity;\r\n\r\n\t\tlightenergy += exp(-shadowdist * ShadowDensity) * curdensity * transmittance * LightColor;\r\n\r\n\t\ttransmittance *= 1- (curdensity);\r\n\t\t\r\n\t\t##if 1\r\n\r\n\t\t//Sky Lighting\r\n\t\tshadowdist = 0;\r\n\t\t\r\n\t\tlpos = CurPos + float3(0,0,0.025);\r\n\t\tfloat lsample = PseudoVolumeTexture(Tex, TexSampler, saturate(lpos), XYFrames, numFrames).r;\r\n\t\tshadowdist += lsample;\r\n\t\tlpos = CurPos + float3(0,0,0.05);\r\n\t\tlsample = PseudoVolumeTexture(Tex, TexSampler, saturate(lpos), XYFrames, numFrames).r;\r\n\t\tshadowdist += lsample;\r\n\t\tlpos = CurPos + float3(0,0,0.15);\r\n\t\tlsample = PseudoVolumeTexture(Tex, TexSampler, saturate(lpos), XYFrames, numFrames).r;\r\n\t\tshadowdist += lsample;\r\n\r\n\t\tlightenergy    += exp(-shadowdist  * AmbientDensity) * curdensity * SkyColor * transmittance;\r\n\t\t##endif\r\n\r\n\t}\r\n\t\t\r\n\r\n\tCurPos -= localcamvec;\r\n}\r\n\r\n\r\n\r\nreturn float4( lightenergy, transmittance);\r\n"
   description=Density RayMarch
   outputType=CMOT_Float4
   inputs [
    {
     inputName=Tex
     input {expression=&self:MaterialExpressionTextureObjectParameter_1}
    }
    {
     inputName=XYFrames
     input {expression=&self:MaterialExpressionComponentMask_2}
    }
    {
     inputName=StepSize
     input {expression=&self:MaterialExpressionDivide_8}
    }
    {
     inputName=MaxSteps
     input {expression=&self:MaterialExpressionClamp_0}
    }
    {
     inputName=FinalStep
     input {expression=&self:MaterialExpressionFrac_3}
    }
    {
     inputName=CurPos
     input {expression=&self:MaterialExpressionMultiply_0}
    }
    {
     inputName=Density
     input {expression=&self:MaterialExpressionScalarParameter_0}
    }
    {
     inputName=ShadowSteps
     input {expression=&self:MaterialExpressionScalarParameter_25}
    }
    {
     inputName=ShadowDensity
     input {expression=&self:MaterialExpressionScalarParameter_26}
    }
    {
     inputName=LightVector
     input {expression=&self:MaterialExpressionNormalize_1}
    }
    {
     inputName=ShadowThreshold
     input {expression=&self:MaterialExpressionScalarParameter_12}
    }
    {
     inputName=LightColor
     input {expression=&self:MaterialExpressionMultiply_17}
    }
    {
     inputName=SkyColor
     input {expression=&self:MaterialExpressionMultiply_19}
    }
    {
     inputName=AmbientDensity
     input {expression=&self:MaterialExpressionScalarParameter_4}
    }
    {
     inputName=Jitter
     input {expression=&self:MaterialExpressionScalarParameter_13}
    }
   ]
  }
  ComponentMask(None=@n5) {b=true, g=true, r=true}
  OneMinus(None=ComponentMask(None=@n5) {a=true})
 ]
}
