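# Unlit alpha-composite material that raymarches an RGBA volume stored as a tiled-slice 2D texture, with depth test disabled.
# VolumeBoxIntersect supplies ray entry and step count to a Custom HLSL loop, which blends adjacent slices and accumulates color and transmittance, using DitherTemporalAA start jitter.
# Squared LightColor tints the color, and inverted transmittance gives opacity.
# No shadow rays are marched, and the FinalStep input goes unused in the loop.
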Material /Volumetrics/Content/VolumeTextures/Materials/Raymarching/M_VolumeRayMarch_Unlit_RGBA
graph {
 properties {bDisableDepthTest=true, blendMode=BLEND_AlphaComposite, pixelDepthOffsetMode=PDOM_Legacy, shadingModel=MSM_Unlit, usedShadingModels=Unlit}
 nodes [
  n1 = VectorParameter() {parameterName=LightColor, defaultValue=<rgba 1 1 1 1>}
  MaterialExpressionTextureObjectParameter_1 = TextureObjectParameter() {parameterName=RT, texture=/Volumetrics/Content/VolumeTextures/Textures/RT_Preview}
  MaterialExpressionComponentMask_8 = ComponentMask(None=VectorParameter() {parameterName=XYFrames, defaultValue=<rgba 16 16 0 0>}.RGB) {g=true, r=true}
  MaterialExpressionDivide_0 = Divide(B=ScalarParameter() {defaultValue=128, parameterName=Steps})
  n2 = MaterialFunctionCall!(
   Steps=ScalarParameter() {defaultValue=128, parameterName=Steps}
   Plane alignment.=ScalarParameter() {defaultValue=1, parameterName=Plane Alignment}
  ) {materialFunction=/Volumetrics/Content/VolumeTextures/MaterialFunctions/VolumeBoxIntersect}
  n3 = Multiply(A=@n2.Box Distance, B=ScalarParameter() {defaultValue=128, parameterName=Steps})
  MaterialExpressionClamp_1 = Clamp(None=Floor(None=@n3)) {desc=MaxSteps, maxDefault=512}
  MaterialExpressionFrac_0 = Frac(None=@n3) {desc=FinalStepSize}
  MaterialExpressionMultiply_5 = Multiply(A=@n2.Intersection Position) {desc=Ray Entry}
  MaterialExpressionScalarParameter_8 = ScalarParameter() {defaultValue=128, parameterName=Density}
  MaterialExpressionMultiply_9 = Multiply(
   A=ScalarParameter() {defaultValue=1, parameterName=Jitter}
   B=MaterialFunctionCall() {materialFunction=&DitherTemporalAA}.Result
  )
  MaterialExpressionComponentMask_12 = ComponentMask(None=VectorParameter() {parameterName=UVScale, defaultValue=<rgba 1 1 0 0>}.RGB) {g=true, r=true}
  n4 = Custom!(
   Tex=@MaterialExpressionTextureObjectParameter_1
   XYFrames=@MaterialExpressionComponentMask_8
   StepSize=@MaterialExpressionDivide_0
   MaxSteps=@MaterialExpressionClamp_1
   FinalStep=@MaterialExpressionFrac_0
   CurPos=@MaterialExpressionMultiply_5
   Density=@MaterialExpressionScalarParameter_8
   Jitter=@MaterialExpressionMultiply_9
   UVScale=@MaterialExpressionComponentMask_12
  ) {
   code="\r\nfloat numFrames = XYFrames.x * XYFrames.y;\r\nfloat curdensity = 0;\r\nfloat transmittance = 1;\r\nfloat3 localcamvec = normalize( mul(Parameters.CameraVector, (float3x3)WSDemote(GetWorldToLocal(Parameters))) ) * StepSize;\r\n\r\nDensity *= StepSize;\r\nfloat3 lightenergy = 0;\r\n\r\nCurPos +=  localcamvec * Jitter;\r\n\r\n\r\nfor (int i = 0; i < MaxSteps; i++)\r\n{\t\r\n\tfloat4 cursample = DensityRaymarch(Tex, TexSampler, saturate(CurPos), XYFrames, UVScale, numFrames);\r\n\t\r\n\t//Sample Light Absorption and Scattering\r\n\tif( cursample.a > 0)\r\n\t{\r\n\t\r\n\t\tcurdensity = saturate((cursample.a) * Density);\r\n\t\tlightenergy += curdensity * transmittance * cursample.rgb;\r\n\t\t\r\n\t\ttransmittance *= 1-curdensity;\r\n\r\n\t}\r\n\t\t\r\n\r\n\tCurPos -= localcamvec;\r\n}\r\n\r\n\r\nreturn float4( lightenergy, transmittance);\r\n"
   desc="\n*** Original source before expression upgrade ***\n\r\nfloat numFrames = XYFrames.x * XYFrames.y;\r\nfloat curdensity = 0;\r\nfloat transmittance = 1;\r\nfloat3 localcamvec = normalize( mul(Parameters.CameraVector, (float3x3)GetPrimitiveData(Parameters).WorldToLocal) ) * StepSize;\r\n\r\nDensity *= StepSize;\r\nfloat3 lightenergy = 0;\r\n\r\nCurPos +=  localcamvec * Jitter;\r\n\r\n\r\nfor (int i = 0; i < MaxSteps; i++)\r\n{\t\r\n\tfloat4 cursample = CustomExpression0(Parameters, Tex, TexSampler, saturate(CurPos), XYFrames, UVScale, numFrames);\r\n\t\r\n\t//Sample Light Absorption and Scattering\r\n\tif( cursample.a > 0)\r\n\t{\r\n\t\r\n\t\tcurdensity = saturate((cursample.a) * Density);\r\n\t\tlightenergy += curdensity * transmittance * cursample.rgb;\r\n\t\t\r\n\t\ttransmittance *= 1-curdensity;\r\n\r\n\t}\r\n\t\t\r\n\r\n\tCurPos -= localcamvec;\r\n}\r\n\r\n\r\nreturn float4( lightenergy, transmittance);\r\n"
   description=Density RayMarch
   outputType=CMOT_Float4
   preamble="MaterialFloat4 DensityRaymarch(\r\n    Texture2D Tex,\r\n    SamplerState TexSampler,\r\n    MaterialFloat3 inPos,\r\n    MaterialFloat2 xsize,\r\n    MaterialFloat2 UVScale,\r\n    MaterialFloat numframes\r\n)\r\n{\r\n    MaterialFloat zframe = ceil(inPos.z * numframes);\r\n    MaterialFloat zphase = frac(inPos.z * numframes);\r\n\r\n    MaterialFloat2 uv = frac(inPos.xy) / xsize;\r\n\r\n    MaterialFloat2 xyidx = 0.0;\r\n\r\n    xyidx.y = floor(zframe / xsize.x);\r\n    xyidx.x = zframe - xsize.x * xyidx.y;\r\n    MaterialFloat2 curframe = xyidx / xsize;\r\n\r\n    xyidx.y = floor((zframe + 1.0) / xsize.x);\r\n    xyidx.x = zframe + 1.0 - xsize.x * xyidx.y;\r\n    MaterialFloat2 nextframe = xyidx / xsize;\r\n\r\n    MaterialFloat4 sampleA = Tex.SampleLevel(\r\n        TexSampler,\r\n        (uv + curframe) * UVScale,\r\n        0\r\n    );\r\n\r\n    MaterialFloat4 sampleB = Tex.SampleLevel(\r\n        TexSampler,\r\n        (uv + nextframe) * UVScale,\r\n        0\r\n    );\r\n\r\n    return lerp(sampleA, sampleB, zphase);\r\n}"
   inputs [
    {
     inputName=Tex
     input {expression=&self:MaterialExpressionTextureObjectParameter_1}
    }
    {
     inputName=XYFrames
     input {expression=&self:MaterialExpressionComponentMask_8}
    }
    {
     inputName=StepSize
     input {expression=&self:MaterialExpressionDivide_0}
    }
    {
     inputName=MaxSteps
     input {expression=&self:MaterialExpressionClamp_1}
    }
    {
     inputName=FinalStep
     input {expression=&self:MaterialExpressionFrac_0}
    }
    {
     inputName=CurPos
     input {expression=&self:MaterialExpressionMultiply_5}
    }
    {
     inputName=Density
     input {expression=&self:MaterialExpressionScalarParameter_8}
    }
    {
     inputName=Jitter
     input {expression=&self:MaterialExpressionMultiply_9}
    }
    {
     inputName=UVScale
     input {expression=&self:MaterialExpressionComponentMask_12}
    }
   ]
  }
  Multiply(A=Multiply(A=@n1.RGB, B=@n1.RGB), B=ComponentMask(None=@n4) {b=true, g=true, r=true})
  OneMinus(None=ComponentMask(None=@n4) {a=true})
 ]
}
